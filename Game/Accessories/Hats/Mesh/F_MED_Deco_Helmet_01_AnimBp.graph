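# Animation Blueprint for the Deco Helmet 01 hat accessory, adding secondary physics motion to two helmet bones.
# An input pose runs through chained Anim Dynamics nodes on dyn_helmet_01 and dyn_helmet_01_a, each angle-limited and spring-driven in actor space.
# A Copy Pose From Mesh node reads MeshToCopy, set at initialization via GetOwnerSkeletalMeshComponent, but is not wired into the output chain.
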AnimBlueprint /Game/Accessories/Hats/Mesh/F_MED_Deco_Helmet_01_AnimBp
target_skeleton=&/F_MED_Deco_Helmet_01_Skeleton
graphs [
 AnimGraph {nodes [
   Copy Pose From Mesh(SourceMeshComponent=Get MeshToCopy() -> (MeshToCopy:Skeletal Mesh Component Object Reference))
   AnimGraphNode_SubInput_0 = Input Pose()
   n1 = Anim Dynamics - Bone: dyn_helmet_01(ComponentPose=Local To Component(LocalPose=@AnimGraphNode_SubInput_0), Alpha=0.5, ExternalForce="0,0,0") {
    alpha=0.9
    angularSpringConstant=1500
    bGravityOverrideInSimSpace=true
    numSolverIterationsPostUpdate=4
    numSolverIterationsPreUpdate=16
    simulationSpace=Actor
    alphaBoolBlend {blendOption=Linear}
    boundBone {boneName=dyn_helmet_01}
    physicsBodyDefinitions [{
      sphereCollisionRadius=0
      boundBone {boneName=dyn_helmet_01}
      boxExtents=<30 30 10>
      constraintSetup {angularTargetAxis=AxisZ, twistAxis=AxisZ, angularLimitsMax=<10 20 0>, angularLimitsMin=<-20 -20 0>, angularTarget=<0 0 1>}
      localJointOffset=<0 0 -3>
     }]
   }
   n2 = Anim Dynamics - Bone: dyn_helmet_01_a(ComponentPose=@n1, Alpha=0.45, ExternalForce="0,0,0") {
    alpha=0.9
    angularSpringConstant=1500
    bGravityOverrideInSimSpace=true
    numSolverIterationsPostUpdate=4
    numSolverIterationsPreUpdate=16
    simulationSpace=Actor
    alphaBoolBlend {blendOption=Linear}
    boundBone {boneName=dyn_helmet_01_a}
    physicsBodyDefinitions [{
      sphereCollisionRadius=0
      boundBone {boneName=dyn_helmet_01_a}
      boxExtents=<30 30 10>
      constraintSetup {angularTargetAxis=AxisZ, twistAxis=AxisZ, angularLimitsMax=<20 30 0>, angularLimitsMin=<-30 -30 0>, angularTarget=<0 0 1>}
      localJointOffset=<0 0 -3>
     }]
   }
   Output Pose(Result=Component To Local(ComponentPose=@n2)) {layerGroup=None}
  ]}
 EventGraph {nodes [
   Event BlueprintInitializeAnimation()
   > Set MeshToCopy(MeshToCopy:Skeletal Mesh Component Object Reference=GetOwnerSkeletalMeshComponent()) -> (Output_Get:Skeletal Mesh Component Object Reference)
   Event BlueprintUpdateAnimation()
  ]}
]
variables [MeshToCopy {type=SkeletalMeshComponent}]
