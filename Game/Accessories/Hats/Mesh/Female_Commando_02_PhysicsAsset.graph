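# Physics asset for the Female_Commando_02 hat accessory mesh, giving it simple collision.
# It holds a single body on the root bone, shaped as one capsule tilted onto its side.
# It defines no constraints, so there are no jointed parts.
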PhysicsAsset /Game/Accessories/Hats/Mesh/Female_Commando_02_PhysicsAsset
bodies {root [{shapeType=Capsule, length=11.3004, radius=7.087, center=<-0.0672 4.6483 9.7665>, rotation=<rot 0 -0.1275 -84.8926>}]}
constraints []
