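# Physics asset for the Male Commando Helmet BR043 hat accessory mesh.
# It holds a single body on the root bone, shaped as one offset, rotated capsule that approximates the helmet's volume.
# It has no constraints, so the helmet simulates or collides as one rigid piece.
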PhysicsAsset /Game/Accessories/Hats/Mesh/Male_Commando_Helmet_BR043_PhysicsAsset
bodies {root [{shapeType=Capsule, length=13.6414, radius=10.7785, center=<-0.3817 3.5492 5.2451>, rotation=<rot 0 -3.1538 -57.6748>}]}
constraints []
