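# Animation Blueprint for the Male Commando Pizza Pit hat accessory skeleton, set up for jiggle on its eye bones.
# Chained Anim Dynamics nodes for eye_l and eye_r define world-space simulation with small linear limits and locked rotation.
# That chain is not wired to the output, which only passes the input pose through space conversions.
# Initialization stores the owner's skeletal mesh in MeshToCopy, which no shown node reads.
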AnimBlueprint /Game/Accessories/Hats/Mesh/Male_Commando_Pizza_Pit_Dyn_AnimBp
target_skeleton=&/Male_Commando_Pizza_Pit_Skeleton
graphs [
 AnimGraph {nodes [
   n1 = Anim Dynamics - Bone: eye_l(Alpha=1, ExternalForce="0,0,0") {
    bGravityOverrideInSimSpace=true
    simulationSpace=World
    alphaBoolBlend {blendOption=Linear}
    boundBone {boneName=eye_l}
    physicsBodyDefinitions [{
      sphereCollisionRadius=0
      boundBone {boneName=eye_l}
      boxExtents=<1 1 1>
      constraintSetup {
       twistAxis=AxisY
       angularLimitsMax=<0 0 0>
       angularLimitsMin=<0 0 0>
       angularTarget=<0 0 0>
       linearAxesMax=<0.1 0 1>
       linearAxesMin=<-0.1 0 -1>
      }
      localJointOffset=<-0.1 0 0>
     }]
   }
   Anim Dynamics - Bone: eye_r(ComponentPose=@n1, Alpha=1, ExternalForce="0,0,0") {
    bGravityOverrideInSimSpace=true
    simulationSpace=World
    alphaBoolBlend {blendOption=Linear}
    boundBone {boneName=eye_r}
    physicsBodyDefinitions [{
      sphereCollisionRadius=0
      boundBone {boneName=eye_r}
      boxExtents=<1 1 1>
      constraintSetup {
       twistAxis=AxisY
       angularLimitsMax=<0 0 0>
       angularLimitsMin=<0 0 0>
       angularTarget=<0 0 0>
       linearAxesMax=<0.1 0 1>
       linearAxesMin=<-0.1 0 -0.9>
      }
      localJointOffset=<-0.1 0 0>
     }]
   }
   AnimGraphNode_SubInput_0 = Input Pose()
   Output Pose(Result=Component To Local(ComponentPose=Local To Component(LocalPose=@AnimGraphNode_SubInput_0))) {layerGroup=None}
  ]}
 EventGraph {nodes [
   Event BlueprintInitializeAnimation()
   > Set MeshToCopy(MeshToCopy:Skeletal Mesh Component Object Reference=GetOwnerSkeletalMeshComponent()) -> (Output_Get:Skeletal Mesh Component Object Reference)
   Event BlueprintUpdateAnimation()
  ]}
]
variables [MeshToCopy {type=SkeletalMeshComponent}]
